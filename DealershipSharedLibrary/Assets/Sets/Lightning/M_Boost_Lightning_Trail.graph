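# Unlit, two-sided additive material for a lightning boost trail drawn on Niagara ribbons.
# A time-driven flipbook samples two lightning atlases, with dynamic parameters choosing atlas, blending channels and setting contrast.
# Panning flow noise distorts a gradient stripe and scrolling lightning texture that shape the bolts.
# Particle color feeds emissive through MF_FX_EmissiveAdjustments, while particle alpha, lightning, stripe and along-U fades build a separate mask.
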Material /DealershipSharedLibrary/Assets/Sets/Lightning/M_Boost_Lightning_Trail
graph {
 properties {
  bUsedWithNiagaraRibbons=true
  blendMode=BLEND_Additive
  opacityMaskClipValue=0.05
  pixelDepthOffsetMode=PDOM_Legacy
  refractionMethod=RM_PixelNormalOffset
  shadingModel=MSM_Unlit
  twoSided=true
  usedShadingModels=Unlit
 }
 nodes [
  n1 = DynamicParameter() {defaultValue=<rgba 1 1 1 5>, paramNames [TextureSwitch, TextureSwitch, MainTrailTextureSwitch, MainTrailExtraLighting]}
  n2 = TextureCoordinate() {uTiling=2, vTiling=0.4}
  n3 = Constant() {r=1}
  n4 = ScalarParameter() {defaultValue=4, group=FlipBook, parameterName=Columns_02}
  n5 = ScalarParameter() {defaultValue=4, group=FlipBook, parameterName=Rows_02}
  n6 = AppendVector(A=@n4, B=@n5)
  n7 = Frac(None=Multiply(A=Time(), B=ScalarParameter() {defaultValue=1, group=FlipBook, parameterName=Speed}))
  n8 = Multiply(A=AppendVector(A=Multiply(A=@n4, B=@n5), B=@n5), B=AppendVector(A=@n7, B=@n7))
  n9 = Add(A=Divide(A=Panner(Coordinate=@n2, Time=@n3) {speedY=0.3}, B=@n6), B=Divide(A=Floor(None=@n8), B=@n6))
  n10 = TextureSample(UVs=@n9) {samplerType=SAMPLERTYPE_LinearColor, texture=&/T_VCVFX_lightningFB_Set_V8}
  n11 = Power(Base=Multiply(A=@n10.R) {constB=1.5}, Exp=@n1.MainTrailExtraLighting) {constExponent=1}
  n12 = TextureSample(UVs=@n9) {samplerType=SAMPLERTYPE_LinearColor, texture=&/T_VCVFX_lightningFB_Set}
  n13 = If(A=@n1.MainTrailTextureSwitch, "A > B"=@n11, "A == B"=@n11, "A < B"=Power(Base=@n12.R, Exp=@n1.MainTrailExtraLighting) {constExponent=1}) {constB=1}
  n14 = OneMinus(None=ComponentMask(None=TextureCoordinate()) {r=true})
  n15 = Power(Base=Power(Base=Multiply(A=@n10.B, B=@n14)) {constExponent=0.5}, Exp=@n1.MainTrailExtraLighting) {constExponent=1}
  n16 = LinearInterpolate(
   A=@n13
   B=If(A=@n1.MainTrailTextureSwitch, "A > B"=@n15, "A == B"=@n15, "A < B"=Power(Base=@n12.B, Exp=@n1.MainTrailExtraLighting) {constExponent=1}) {constB=1}
   Alpha=@n1.TextureSwitch
  )
  n17 = LinearInterpolate(
   A=@n13
   B=Multiply(A=If(A=@n1.MainTrailTextureSwitch, "A > B"=@n10.G, "A == B"=@n10.G, "A < B"=@n12.G) {constB=1}, B=Constant() {r=1})
   Alpha=@n1.TextureSwitch
  )
  n18 = Panner(Coordinate=TextureCoordinate(), Time=Multiply(A=Time(), B=Constant() {r=0.5})) {speedX=-2}
  n19 = Multiply(
   A=ComponentMask(None=TextureSample(UVs=@n18) {texture=/DealershipSharedLibrary/FX/Textures/Noise/T_VCVFX_Flow03_N}.RGB) {g=true, r=true}
   B=Constant() {r=0.2}
  )
  n20 = TextureSample(UVs=Add(A=TextureCoordinate() {uTiling=2}, B=@n19)) {texture=/DealershipSharedLibrary/FX/Textures/Gradient_Masks/T_VCVFX_GradientStripe_A}
  n21 = Multiply(A=Power(Base=Add(A=@n16, B=@n17)), B=Multiply(A=@n20.G) {constB=2})
  MaterialExpressionVertexColor_8 = ParticleColor()
  n22 = Add(A=Multiply(A=Power(Base=@n21)) {constB=0.35}, B=Multiply(A=@MaterialExpressionVertexColor_8.RGBA) {constB=0.25})
  MaterialFunctionCall(Emissive=If(A=@n1.MainTrailTextureSwitch, "A > B"=@n22, "A == B"=@n22, "A < B"=@MaterialExpressionVertexColor_8.RGBA) {constB=1}) {materialFunction=&MF_FX_EmissiveAdjustments}
  n23 = OneMinus(None=ComponentMask(None=TextureCoordinate()) {r=true})
  n24 = Multiply(A=Multiply(A=@n23, B=@n20.R), B=Constant() {r=1})
  Multiply(A=@n24, B=@n24)
  n25 = Multiply(A=@n23, B=@n23)
  n26 = Power(Base=Multiply(A=Multiply(A=@n20.R, B=@n25), B=Constant() {r=8})) {constExponent=0.5}
  n27 = TextureCoordinate() {uTiling=2}
  n28 = TextureSample(UVs=Add(A=@n19, B=Panner(Coordinate=@n27) {speedX=-1})) {texture=&/T_VCVFX_Lightning2}
  n29 = LinearInterpolate(A=@n28.R, B=@n28.R, Alpha=@n1.TextureSwitch)
  n30 = Constant() {r=1}
  n31 = Add(A=Add(A=Multiply(A=@n29, B=@n30), B=Power(Base=@n16) {constExponent=1}), B=@n17)
  n32 = ComponentMask(None=TextureCoordinate()) {r=true}
  n33 = Multiply(A=Multiply(A=@n32, B=@n32), B=OneMinus(None=@n32))
  Multiply(A=Multiply(A=@MaterialExpressionVertexColor_8.A, B=Multiply(A=@n26, B=@n31)), B=Multiply(A=Power(Base=@n33) {constExponent=0.5}) {constB=5}) {constA=5}
 ]
}
