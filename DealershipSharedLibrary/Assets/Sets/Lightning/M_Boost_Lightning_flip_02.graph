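# Additive, unlit Niagara sprite material that draws animated lightning, named for a boost effect.
# A rotating noise normal distorts the UVs of panning water noise and two lightning textures, which multiply with a panning gradient mask.
# Dynamic parameters blend mask motion, blend in a fixed quarter-turn rotation, switch lightning textures and scale brightness.
# Particle color tints the result through an emissive adjustment function, and particle alpha scales the unclamped mask.
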Material /DealershipSharedLibrary/Assets/Sets/Lightning/M_Boost_Lightning_flip_02
graph {
 properties {bUsedWithNiagaraSprites=true, blendMode=BLEND_Additive, pixelDepthOffsetMode=PDOM_Legacy, shadingModel=MSM_Unlit, usedShadingModels=Unlit}
 nodes [
  MaterialExpressionVertexColor_7 = ParticleColor()
  n1 = TextureCoordinate() {uTiling=0.25, vTiling=0.25}
  n2 = DynamicParameter() {paramNames [DetailMotion, Brightness, RotateTexture, TextureSwitch]}
  n3 = LinearInterpolate(A=Panner(Coordinate=@n1) {speedX=0.25, speedY=-0.25}, B=Panner(Coordinate=@n1) {speedX=0.25, speedY=-0.15}, Alpha=@n2.DetailMotion)
  n4 = TextureCoordinate() {uTiling=2.5, vTiling=2.5}
  n5 = TextureCoordinate() {uTiling=0.25, vTiling=0.25}
  n6 = Multiply(A=Time(), B=Constant() {r=0.5})
  n7 = ComponentMask(None=TextureSample(UVs=Rotator(Coordinate=@n5, Time=@n6) {speed=-0.2}) {texture=/DealershipSharedLibrary/FX/Textures/Noise/T_VCVFX_NoiseFire01_N}.RGB) {g=true, r=true}
  n8 = Multiply(A=@n7, B=Constant() {r=0.05})
  n9 = TextureSample(UVs=Panner(Coordinate=Add(A=@n4, B=@n8)) {speedX=0.1, speedY=0.5}) {texture=/DealershipSharedLibrary/FX/Textures/Noise/T_VCVFX_Water01_A}
  n10 = Constant() {r=5.5}
  MaterialExpressionConstantClamp_13 = Saturate(None=Add(A=Multiply(A=@n9.RGB, B=@n10), B=Constant() {r=0.1}))
  n11 = Multiply(A=@n7, B=Constant() {r=0.025})
  n12 = Add(A=@n11, B=TextureCoordinate() {uTiling=2, vTiling=2})
  n13 = Constant() {r=0.5}
  n14 = TextureSample(UVs=LinearInterpolate(A=Rotator(Coordinate=@n12, Time=@n13) {speed=3.14}, B=@n12, Alpha=@n2.RotateTexture)) {texture=&/T_VCVFX_Lightning}
  n15 = Constant() {r=2}
  n16 = Add(A=@n11, B=TextureCoordinate())
  n17 = Constant() {r=0.5}
  n18 = TextureSample(UVs=LinearInterpolate(A=Rotator(Coordinate=@n16, Time=@n17) {speed=3.14}, B=@n16, Alpha=@n2.RotateTexture)) {texture=/DealershipSharedLibrary/FX/Textures/Gradient_Masks/T_VCVFX_LightningFlip_A}
  n19 = Constant() {r=10}
  n20 = Multiply(
   A=Multiply(
    A=TextureSample(UVs=@n3) {texture=/DealershipSharedLibrary/FX/Textures/Gradient_Masks/T_VCVFX_WidgetGridVertexColorMaterial_A}.RGB
    B=@MaterialExpressionConstantClamp_13
   )
   B=LinearInterpolate(A=Multiply(A=@n14.B, B=@n15), B=Multiply(A=@n18.R, B=@n19), Alpha=@n2.TextureSwitch)
  )
  MaterialExpressionConstantClamp_12 = Saturate(None=@n20)
  MaterialFunctionCall(Emissive=Multiply(A=Multiply(A=@MaterialExpressionVertexColor_7.RGB, B=@MaterialExpressionConstantClamp_12), B=@n2.Brightness)) {materialFunction=&MF_FX_EmissiveAdjustments}
  Multiply(A=@MaterialExpressionVertexColor_7.A, B=@n20)
 ]
}
